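# Material function that applies an Exclusion blend between a Base and a Blend input, exposed to the library under Misc and Blends.
# It offsets both inputs by -0.5, multiplies them, doubles the product and subtracts that from 0.5.
# Default engine blend textures feed the input previews.
# The output's description says it adds base and blend, which does not match this math.
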MaterialFunction /Engine/Functions/Engine_MaterialFunctions03/Blends/Blend_Exclusion
graph {
 properties {
  bExposeToLibrary=true
  description="Halves the Blend and Base, combines them together, then partially inverts the results."
  libraryCategoriesText [Misc,Blends]
 }
 nodes [
  n1 = TextureSample() {texture=/Engine/EngineMaterials/BlendFunc_DefBase}
  n2 = Constant() {r=-0.5}
  n3 = TextureSample() {texture=/Engine/EngineMaterials/BlendFunc_DefBlend}
  n4 = Multiply(
   A=Add(A=FunctionInput(Preview=@n1.RGB) {description=Base texture input, inputName=Base}, B=@n2)
   B=Add(A=FunctionInput(Preview=@n3.RGB) {description=Blend texture input, inputName=Blend}, B=@n2)
  )
  n5 = Constant() {r=2}
  FunctionOutput(None=Subtract(A=Constant() {r=0.5}, B=Multiply(A=@n4, B=@n5))) {description=Adds the base and blend together.}
 ]
}
